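# Level Sequence for a battle-stage establishing shot in the Pilgrim camera manager, cutting to one spawned cine camera for about 20 seconds.
# A spawned relocator actor yaws half a turn, apparently as the pivot for the attach tracks on the camera and LookAt actor.
# The camera animates position and focal length, with look-at tracking and an actor reference track, while the LookAt actor glides to its local origin.
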LevelSequence /PilgrimCore/CameraManager/Sequences/BattleStage/EstablishingShots/LS_PilgrimCamera_ES_01
bindings [
 CameraComponent {
  id=#1
  parent_id=#2
  possessed_class=CineCameraComponent
  tracks [
   {
    type=MovieSceneFloatTrack
    display_name=Current Focal Length
    sections [{
      type=MovieSceneFloatSection
      channels [None {type=MovieSceneScriptingFloatChannel, default=20.9078, keys [[-7,20.9078],[434,26.2332],[649,21.6991]]}]
     }]
   }
   {
    type=MovieSceneFloatTrack
    display_name="Manual Focus Distance (Focus Settings)"
    sections [{type=MovieSceneFloatSection, channels [None {type=MovieSceneScriptingFloatChannel, default=100000}]}]
   }
   {
    type=MovieSceneFloatTrack
    display_name=Current Aperture
    sections [{type=MovieSceneFloatSection, channels [None {type=MovieSceneScriptingFloatChannel, default=2.8}]}]
   }
  ]
 }
 Cine Camera Actor {
  id=#2
  possessed_class=CineCameraActor
  tracks [
   {type=MovieSceneSpawnTrack, sections [{type=MovieSceneSpawnSection, channels [None {type=MovieSceneScriptingBoolChannel, default=true}]}]}
   {
    type=MovieScene3DTransformTrack
    sections [{
      type=MovieScene3DTransformSection
      channels [
       Location.X {type=MovieSceneScriptingDoubleChannel, default=3922, keys [[1,3922],[434,2899],[649,3000]]}
       Location.Y {type=MovieSceneScriptingDoubleChannel, default=-1363, keys [[1,-1363],[434,-2779],[649,-2425.5273]]}
       Location.Z {type=MovieSceneScriptingDoubleChannel, default=1604, keys [[1,1604],[434,2000],[649,2812.7019]]}
       Rotation.X {type=MovieSceneScriptingDoubleChannel, default=0, const [0,3]}
       Rotation.Y {type=MovieSceneScriptingDoubleChannel, default=-18.9382, const [-22.5769,3]}
       Rotation.Z {type=MovieSceneScriptingDoubleChannel, default=-169.0493, const [-201.0113,3]}
       Scale.X {type=MovieSceneScriptingDoubleChannel, default=1, const [1,3]}
       Scale.Y {type=MovieSceneScriptingDoubleChannel, default=1, const [1,3]}
       Scale.Z {type=MovieSceneScriptingDoubleChannel, default=1, const [1,3]}
      ]
     }]
   }
   {
    type=MovieScene3DAttachTrack
    sections [{type=MovieScene3DAttachSection, end_frame=603, start_frame=602}, {type=MovieScene3DAttachSection, end_frame=602, start_frame=0}]
   }
   {
    type=MovieSceneBoolTrack
    display_name="Enable Look at Tracking (Lookat Tracking Settings)"
    sections [{type=MovieSceneBoolSection, channels [None {type=MovieSceneScriptingBoolChannel, default=true}]}]
   }
   {type=MovieSceneActorReferenceTrack}
  ]
 }
 relocator {
  id=#3
  possessed_class=Actor
  tracks [
   {
    type=MovieSceneSpawnTrack
    sections [
     {type=MovieSceneSpawnSection, channels [None {type=MovieSceneScriptingBoolChannel, default=true}]}
     {type=MovieSceneSpawnSection, channels [None {type=MovieSceneScriptingBoolChannel, default=true}]}
    ]
   }
   {
    type=MovieScene3DTransformTrack
    sections [{
      type=MovieScene3DTransformSection
      channels [
       Location.X {type=MovieSceneScriptingDoubleChannel, default=0}
       Location.Y {type=MovieSceneScriptingDoubleChannel, default=0}
       Location.Z {type=MovieSceneScriptingDoubleChannel, default=0}
       Rotation.X {type=MovieSceneScriptingDoubleChannel, default=0, const [0,1]}
       Rotation.Y {type=MovieSceneScriptingDoubleChannel, default=0, const [0,1]}
       Rotation.Z {type=MovieSceneScriptingDoubleChannel, default=180, keys [[0,0],[601,180]]}
       Scale.X {type=MovieSceneScriptingDoubleChannel, default=1}
       Scale.Y {type=MovieSceneScriptingDoubleChannel, default=1}
       Scale.Z {type=MovieSceneScriptingDoubleChannel, default=1}
      ]
     }]
   }
  ]
 }
 LookAt {
  id=#4
  possessed_class=Actor
  tracks [
   {
    type=MovieSceneSpawnTrack
    sections [
     {type=MovieSceneSpawnSection, channels [None {type=MovieSceneScriptingBoolChannel, default=true}]}
     {type=MovieSceneSpawnSection, channels [None {type=MovieSceneScriptingBoolChannel, default=true}]}
    ]
   }
   {
    type=MovieScene3DTransformTrack
    sections [{
      type=MovieScene3DTransformSection
      channels [
       Location.X {type=MovieSceneScriptingDoubleChannel, default=0, keys [[-73,-2257.3362],[434,0]]}
       Location.Y {type=MovieSceneScriptingDoubleChannel, default=0, keys [[-73,173.5786],[434,0]]}
       Location.Z {type=MovieSceneScriptingDoubleChannel, default=0, const [0,2]}
       Rotation.X {type=MovieSceneScriptingDoubleChannel, default=0, const [0,2]}
       Rotation.Y {type=MovieSceneScriptingDoubleChannel, default=0, const [0,2]}
       Rotation.Z {type=MovieSceneScriptingDoubleChannel, default=0, const [0,2]}
       Scale.X {type=MovieSceneScriptingDoubleChannel, default=1, const [1,2]}
       Scale.Y {type=MovieSceneScriptingDoubleChannel, default=1, const [1,2]}
       Scale.Z {type=MovieSceneScriptingDoubleChannel, default=1, const [1,2]}
      ]
     }]
   }
   {
    type=MovieScene3DAttachTrack
    sections [{type=MovieScene3DAttachSection, end_frame=603, start_frame=602}, {type=MovieScene3DAttachSection, end_frame=602, start_frame=0}]
   }
  ]
 }
]
display_rate {denominator=1, numerator=30}
tick_resolution {denominator=1, numerator=24000}
tracks [{type=MovieSceneCameraCutTrack, sections [{type=MovieSceneCameraCutSection, end_frame=602, start_frame=0}]}]
