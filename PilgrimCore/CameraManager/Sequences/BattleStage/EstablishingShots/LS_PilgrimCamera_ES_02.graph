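# Level sequence for the second Pilgrim battle stage establishing shot, a single slow, sweeping cine camera move.
# A spawned pivot actor yaws half a turn while the attached camera's location descends and its focal length widens slightly.
# Look-at tracking with an actor reference handles aim, and a second static helper actor has its own attach track.
# Attach and look-at targets are not recorded here; one camera cut spans the shot.
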LevelSequence /PilgrimCore/CameraManager/Sequences/BattleStage/EstablishingShots/LS_PilgrimCamera_ES_02
bindings [
 CameraComponent {
  id=#1
  parent_id=#2
  possessed_class=CineCameraComponent
  tracks [
   {
    type=MovieSceneFloatTrack
    display_name=Current Focal Length
    sections [{type=MovieSceneFloatSection, channels [None {type=MovieSceneScriptingFloatChannel, default=10.4579, keys [[0,10.4579],[863,9.2269]]}]}]
   }
   {
    type=MovieSceneFloatTrack
    display_name="Manual Focus Distance (Focus Settings)"
    sections [{type=MovieSceneFloatSection, channels [None {type=MovieSceneScriptingFloatChannel, default=100000}]}]
   }
   {
    type=MovieSceneFloatTrack
    display_name=Current Aperture
    sections [{type=MovieSceneFloatSection, channels [None {type=MovieSceneScriptingFloatChannel, default=2.8}]}]
   }
  ]
 }
 CineCameraActor0 {
  id=#2
  possessed_class=CineCameraActor
  tracks [
   {type=MovieSceneSpawnTrack, sections [{type=MovieSceneSpawnSection, channels [None {type=MovieSceneScriptingBoolChannel, default=true}]}]}
   {
    type=MovieScene3DTransformTrack
    sections [{
      type=MovieScene3DTransformSection
      channels [
       Location.X {type=MovieSceneScriptingDoubleChannel, default=4966, keys [[-45,6801.9121],[941,4966]]}
       Location.Y {type=MovieSceneScriptingDoubleChannel, default=13501.3008, keys [[-45,-4175.3979],[941,-2425.5273]]}
       Location.Z {type=MovieSceneScriptingDoubleChannel, default=3772, keys [[-45,4841.8955],[941,3772]]}
       Rotation.X {type=MovieSceneScriptingDoubleChannel, default=0, const [0,2]}
       Rotation.Y {type=MovieSceneScriptingDoubleChannel, default=-18.9382, const [-22.5769,2]}
       Rotation.Z {type=MovieSceneScriptingDoubleChannel, default=-169.0493, const [-201.0113,2]}
       Scale.X {type=MovieSceneScriptingDoubleChannel, default=1, const [1,2]}
       Scale.Y {type=MovieSceneScriptingDoubleChannel, default=1, const [1,2]}
       Scale.Z {type=MovieSceneScriptingDoubleChannel, default=1, const [1,2]}
      ]
     }]
   }
   {type=MovieScene3DAttachTrack, sections [{type=MovieScene3DAttachSection, end_frame=1014, start_frame=-201}]}
   {
    type=MovieSceneBoolTrack
    display_name="Enable Look at Tracking (Lookat Tracking Settings)"
    sections [{type=MovieSceneBoolSection, channels [None {type=MovieSceneScriptingBoolChannel, default=true}]}]
   }
   {type=MovieSceneActorReferenceTrack}
  ]
 }
 Actor0 {
  id=#3
  possessed_class=Actor
  tracks [
   {
    type=MovieSceneSpawnTrack
    sections [
     {type=MovieSceneSpawnSection, channels [None {type=MovieSceneScriptingBoolChannel, default=true}]}
     {type=MovieSceneSpawnSection, channels [None {type=MovieSceneScriptingBoolChannel, default=true}]}
    ]
   }
   {
    type=MovieScene3DTransformTrack
    sections [{
      type=MovieScene3DTransformSection
      channels [
       Location.X {type=MovieSceneScriptingDoubleChannel, default=0}
       Location.Y {type=MovieSceneScriptingDoubleChannel, default=0}
       Location.Z {type=MovieSceneScriptingDoubleChannel, default=0}
       Rotation.X {type=MovieSceneScriptingDoubleChannel, default=0, const [0,1]}
       Rotation.Y {type=MovieSceneScriptingDoubleChannel, default=0, const [0,1]}
       Rotation.Z {type=MovieSceneScriptingDoubleChannel, default=180, keys [[-37,0],[974.03,180]]}
       Scale.X {type=MovieSceneScriptingDoubleChannel, default=1}
       Scale.Y {type=MovieSceneScriptingDoubleChannel, default=1}
       Scale.Z {type=MovieSceneScriptingDoubleChannel, default=1}
      ]
     }]
   }
  ]
 }
 Actor1 {
  id=#4
  possessed_class=Actor
  tracks [
   {
    type=MovieSceneSpawnTrack
    sections [
     {type=MovieSceneSpawnSection, channels [None {type=MovieSceneScriptingBoolChannel, default=true}]}
     {type=MovieSceneSpawnSection, channels [None {type=MovieSceneScriptingBoolChannel, default=true}]}
    ]
   }
   {
    type=MovieScene3DTransformTrack
    sections [{
      type=MovieScene3DTransformSection
      channels [
       Location.X {type=MovieSceneScriptingDoubleChannel, default=0, const [0,1]}
       Location.Y {type=MovieSceneScriptingDoubleChannel, default=0, const [0,1]}
       Location.Z {type=MovieSceneScriptingDoubleChannel, default=0, const [0,1]}
       Rotation.X {type=MovieSceneScriptingDoubleChannel, default=0, const [0,1]}
       Rotation.Y {type=MovieSceneScriptingDoubleChannel, default=0, const [0,1]}
       Rotation.Z {type=MovieSceneScriptingDoubleChannel, default=0, const [0,1]}
       Scale.X {type=MovieSceneScriptingDoubleChannel, default=1, const [1,1]}
       Scale.Y {type=MovieSceneScriptingDoubleChannel, default=1, const [1,1]}
       Scale.Z {type=MovieSceneScriptingDoubleChannel, default=1, const [1,1]}
      ]
     }]
   }
   {
    type=MovieScene3DAttachTrack
    sections [{type=MovieScene3DAttachSection, end_frame=901, start_frame=900}, {type=MovieScene3DAttachSection, end_frame=900, start_frame=0}]
   }
  ]
 }
]
display_rate {denominator=1, numerator=30}
tick_resolution {denominator=1, numerator=24000}
tracks [{type=MovieSceneCameraCutTrack, sections [{type=MovieSceneCameraCutSection, end_frame=961, start_frame=-37}]}]
