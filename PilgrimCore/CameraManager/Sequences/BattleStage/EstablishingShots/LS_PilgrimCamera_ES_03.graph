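# Battle-stage establishing shot 03 for the Pilgrim camera: a Level Sequence with one camera cut to a Cine Camera.
# A relocator actor turns a half circle in yaw over the shot, likely carrying the attached camera and LookAt actor to create an orbit.
# The camera adds keyed positional drift and a widening focal length, with look-at tracking enabled and an actor reference track, presumably targeting LookAt.
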LevelSequence /PilgrimCore/CameraManager/Sequences/BattleStage/EstablishingShots/LS_PilgrimCamera_ES_03
bindings [
 CameraComponent {
  id=#1
  parent_id=#2
  possessed_class=CineCameraComponent
  tracks [
   {
    type=MovieSceneFloatTrack
    display_name=Current Focal Length
    sections [{
      type=MovieSceneFloatSection
      channels [None {type=MovieSceneScriptingFloatChannel, default=3.7682, keys [[-43,10.1213],[602,3.7682],[644,38.1777]]}]
     }]
   }
   {
    type=MovieSceneFloatTrack
    display_name="Manual Focus Distance (Focus Settings)"
    sections [{type=MovieSceneFloatSection, channels [None {type=MovieSceneScriptingFloatChannel, default=100000}]}]
   }
   {
    type=MovieSceneFloatTrack
    display_name=Current Aperture
    sections [{type=MovieSceneFloatSection, channels [None {type=MovieSceneScriptingFloatChannel, default=2.8}]}]
   }
  ]
 }
 Cine Camera Actor {
  id=#2
  possessed_class=CineCameraActor
  tracks [
   {type=MovieSceneSpawnTrack, sections [{type=MovieSceneSpawnSection, channels [None {type=MovieSceneScriptingBoolChannel, default=true}]}]}
   {
    type=MovieScene3DTransformTrack
    sections [{
      type=MovieScene3DTransformSection
      channels [
       Location.X {type=MovieSceneScriptingDoubleChannel, default=3000, keys [[-47,2000],[296,2540],[649,3000],[661,10870.8711]]}
       Location.Y {type=MovieSceneScriptingDoubleChannel, default=-1203, keys [[-47,-1617],[296,-1726],[649,-1203],[661,-4175.3979]]}
       Location.Z {type=MovieSceneScriptingDoubleChannel, default=1966, keys [[-47,1970],[296,2663],[649,1966],[661,4841.8955]]}
       Rotation.X {type=MovieSceneScriptingDoubleChannel, default=0, const [0,4]}
       Rotation.Y {type=MovieSceneScriptingDoubleChannel, default=-18.9382, const [-22.5769,4]}
       Rotation.Z {type=MovieSceneScriptingDoubleChannel, default=-169.0493, const [-201.0113,4]}
       Scale.X {type=MovieSceneScriptingDoubleChannel, default=1, const [1,4]}
       Scale.Y {type=MovieSceneScriptingDoubleChannel, default=1, const [1,4]}
       Scale.Z {type=MovieSceneScriptingDoubleChannel, default=1, const [1,4]}
      ]
     }]
   }
   {
    type=MovieScene3DAttachTrack
    sections [{type=MovieScene3DAttachSection, end_frame=603, start_frame=602}, {type=MovieScene3DAttachSection, end_frame=602, start_frame=0}]
   }
   {
    type=MovieSceneBoolTrack
    display_name="Enable Look at Tracking (Lookat Tracking Settings)"
    sections [{type=MovieSceneBoolSection, channels [None {type=MovieSceneScriptingBoolChannel, default=true}]}]
   }
   {type=MovieSceneActorReferenceTrack}
  ]
 }
 relocator {
  id=#3
  possessed_class=Actor
  tracks [
   {
    type=MovieSceneSpawnTrack
    sections [
     {type=MovieSceneSpawnSection, channels [None {type=MovieSceneScriptingBoolChannel, default=true}]}
     {type=MovieSceneSpawnSection, channels [None {type=MovieSceneScriptingBoolChannel, default=true}]}
    ]
   }
   {
    type=MovieScene3DTransformTrack
    sections [{
      type=MovieScene3DTransformSection
      channels [
       Location.X {type=MovieSceneScriptingDoubleChannel, default=0}
       Location.Y {type=MovieSceneScriptingDoubleChannel, default=0}
       Location.Z {type=MovieSceneScriptingDoubleChannel, default=0}
       Rotation.X {type=MovieSceneScriptingDoubleChannel, default=0, const [0,1]}
       Rotation.Y {type=MovieSceneScriptingDoubleChannel, default=0, const [0,1]}
       Rotation.Z {type=MovieSceneScriptingDoubleChannel, default=180, keys [[0,0],[601,180]]}
       Scale.X {type=MovieSceneScriptingDoubleChannel, default=1}
       Scale.Y {type=MovieSceneScriptingDoubleChannel, default=1}
       Scale.Z {type=MovieSceneScriptingDoubleChannel, default=1}
      ]
     }]
   }
  ]
 }
 LookAt {
  id=#4
  possessed_class=Actor
  tracks [
   {
    type=MovieSceneSpawnTrack
    sections [
     {type=MovieSceneSpawnSection, channels [None {type=MovieSceneScriptingBoolChannel, default=true}]}
     {type=MovieSceneSpawnSection, channels [None {type=MovieSceneScriptingBoolChannel, default=true}]}
    ]
   }
   {
    type=MovieScene3DTransformTrack
    sections [{
      type=MovieScene3DTransformSection
      channels [
       Location.X {type=MovieSceneScriptingDoubleChannel, default=0, const [0,1]}
       Location.Y {type=MovieSceneScriptingDoubleChannel, default=0, const [0,1]}
       Location.Z {type=MovieSceneScriptingDoubleChannel, default=0, const [0,1]}
       Rotation.X {type=MovieSceneScriptingDoubleChannel, default=0, const [0,1]}
       Rotation.Y {type=MovieSceneScriptingDoubleChannel, default=0, const [0,1]}
       Rotation.Z {type=MovieSceneScriptingDoubleChannel, default=0, const [0,1]}
       Scale.X {type=MovieSceneScriptingDoubleChannel, default=1, const [1,1]}
       Scale.Y {type=MovieSceneScriptingDoubleChannel, default=1, const [1,1]}
       Scale.Z {type=MovieSceneScriptingDoubleChannel, default=1, const [1,1]}
      ]
     }]
   }
   {
    type=MovieScene3DAttachTrack
    sections [{type=MovieScene3DAttachSection, end_frame=603, start_frame=602}, {type=MovieScene3DAttachSection, end_frame=602, start_frame=0}]
   }
  ]
 }
]
display_rate {denominator=1, numerator=30}
tick_resolution {denominator=1, numerator=24000}
tracks [{type=MovieSceneCameraCutTrack, sections [{type=MovieSceneCameraCutSection, end_frame=602, start_frame=0}]}]
