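# Level sequence for a battle stage establishing shot in the Pilgrim camera system, cutting to one spawned cine camera for about 20 seconds.
# A relocator actor yaws through most of a turn, apparently as an orbit pivot for the attached camera and LookAt actor.
# The camera follows its own keyed move with look-at tracking enabled, presumably aimed at the moving LookAt actor.
# Focal length, focus distance and aperture stay fixed.
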LevelSequence /PilgrimCore/CameraManager/Sequences/BattleStage/EstablishingShots/LS_PilgrimCamera_ES_06
bindings [
 CameraComponent {
  id=#1
  parent_id=#2
  possessed_class=CineCameraComponent
  tracks [
   {
    type=MovieSceneFloatTrack
    display_name=Current Focal Length
    sections [{type=MovieSceneFloatSection, channels [None {type=MovieSceneScriptingFloatChannel, default=24}]}]
   }
   {
    type=MovieSceneFloatTrack
    display_name="Manual Focus Distance (Focus Settings)"
    sections [{type=MovieSceneFloatSection, channels [None {type=MovieSceneScriptingFloatChannel, default=100000}]}]
   }
   {
    type=MovieSceneFloatTrack
    display_name=Current Aperture
    sections [{type=MovieSceneFloatSection, channels [None {type=MovieSceneScriptingFloatChannel, default=2.8}]}]
   }
  ]
 }
 Cine Camera Actor {
  id=#2
  possessed_class=CineCameraActor
  tracks [
   {type=MovieSceneSpawnTrack, sections [{type=MovieSceneSpawnSection, channels [None {type=MovieSceneScriptingBoolChannel, default=true}]}]}
   {
    type=MovieScene3DTransformTrack
    sections [{
      type=MovieScene3DTransformSection
      channels [
       Location.X {type=MovieSceneScriptingDoubleChannel, default=3000, keys [[-93,6314.9897],[0,4867.6001],[159,3000],[599,3000],[656,6314.9897]]}
       Location.Y {type=MovieSceneScriptingDoubleChannel, default=-2134, keys [[-93,-2425.5273],[0,-2806.0549],[159,-3125],[599,-2134],[656,-2425.5273]]}
       Location.Z {type=MovieSceneScriptingDoubleChannel, default=2131, keys [[-93,-3234],[0,-945.2615],[164,2131],[599,1000],[656,2812.7019]]}
       Rotation.X {type=MovieSceneScriptingDoubleChannel, default=-12, keys [[0,-12],[602.9987,-8]]}
       Rotation.Y {type=MovieSceneScriptingDoubleChannel, default=-18.9382, keys [[-93,-22.5769],[0,-1.1236],[159,-22.5769],[656,-22.5769]]}
       Rotation.Z {type=MovieSceneScriptingDoubleChannel, default=-169.0493, keys [[-93,-201.0113],[0,-179.1859],[159,-201.0113],[656,-201.0113]]}
       Scale.X {type=MovieSceneScriptingDoubleChannel, default=1, const [1,4]}
       Scale.Y {type=MovieSceneScriptingDoubleChannel, default=1, const [1,4]}
       Scale.Z {type=MovieSceneScriptingDoubleChannel, default=1, const [1,4]}
      ]
     }]
   }
   {
    type=MovieScene3DAttachTrack
    sections [{type=MovieScene3DAttachSection, end_frame=603, start_frame=602}, {type=MovieScene3DAttachSection, end_frame=602, start_frame=-61}]
   }
   {
    type=MovieSceneBoolTrack
    display_name="Enable Look at Tracking (Lookat Tracking Settings)"
    sections [{type=MovieSceneBoolSection, channels [None {type=MovieSceneScriptingBoolChannel, default=true}]}]
   }
   {type=MovieSceneActorReferenceTrack}
  ]
 }
 relocator {
  id=#3
  possessed_class=Actor
  tracks [
   {
    type=MovieSceneSpawnTrack
    sections [
     {type=MovieSceneSpawnSection, channels [None {type=MovieSceneScriptingBoolChannel, default=true}]}
     {type=MovieSceneSpawnSection, channels [None {type=MovieSceneScriptingBoolChannel, default=true}]}
    ]
   }
   {
    type=MovieScene3DTransformTrack
    sections [{
      type=MovieScene3DTransformSection
      channels [
       Location.X {type=MovieSceneScriptingDoubleChannel, default=0}
       Location.Y {type=MovieSceneScriptingDoubleChannel, default=0}
       Location.Z {type=MovieSceneScriptingDoubleChannel, default=0}
       Rotation.X {type=MovieSceneScriptingDoubleChannel, default=0}
       Rotation.Y {type=MovieSceneScriptingDoubleChannel, default=0}
       Rotation.Z {type=MovieSceneScriptingDoubleChannel, default=0, keys [[-76,0],[660,319]]}
       Scale.X {type=MovieSceneScriptingDoubleChannel, default=1}
       Scale.Y {type=MovieSceneScriptingDoubleChannel, default=1}
       Scale.Z {type=MovieSceneScriptingDoubleChannel, default=1}
      ]
     }]
   }
  ]
 }
 LookAt {
  id=#4
  possessed_class=Actor
  tracks [
   {
    type=MovieSceneSpawnTrack
    sections [
     {type=MovieSceneSpawnSection, channels [None {type=MovieSceneScriptingBoolChannel, default=true}]}
     {type=MovieSceneSpawnSection, channels [None {type=MovieSceneScriptingBoolChannel, default=true}]}
    ]
   }
   {
    type=MovieScene3DTransformTrack
    sections [{
      type=MovieScene3DTransformSection
      channels [
       Location.Y {type=MovieSceneScriptingDoubleChannel, default=-1634, keys [[-93,-1634],[434,2448]]}
       Location.Z {type=MovieSceneScriptingDoubleChannel, default=-2529, keys [[-93,0],[174,-2529],[434,-380]]}
       Rotation.X {type=MovieSceneScriptingDoubleChannel, default=34.629}
       Rotation.Y {type=MovieSceneScriptingDoubleChannel, default=-5.4332}
       Rotation.Z {type=MovieSceneScriptingDoubleChannel, default=-1.6949}
       Scale.X {type=MovieSceneScriptingDoubleChannel, default=1}
       Scale.Y {type=MovieSceneScriptingDoubleChannel, default=1}
       Scale.Z {type=MovieSceneScriptingDoubleChannel, default=1}
      ]
     }]
   }
   {
    type=MovieScene3DAttachTrack
    sections [{type=MovieScene3DAttachSection, end_frame=603, start_frame=602}, {type=MovieScene3DAttachSection, end_frame=602, start_frame=0}]
   }
  ]
 }
]
display_rate {denominator=1, numerator=30}
tick_resolution {denominator=1, numerator=24000}
tracks [{type=MovieSceneCameraCutTrack, sections [{type=MovieSceneCameraCutSection, end_frame=602, start_frame=0}]}]
